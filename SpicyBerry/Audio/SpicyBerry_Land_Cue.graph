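# Sound cue that plays the landing sound for the SpicyBerry.
# A Random node picks one of three equally weighted landing waves and feeds a Modulator at default settings for slight pitch and volume variation.
# The cue plays at reduced volume, borrowing weapon-drop attenuation and routing to the loot-collect sound class.
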SoundCue /SpicyBerry/Audio/SpicyBerry_Land_Cue
root=Modulator_2
nodes [
 {id=Modulator_2, children [Random_1]}
 {id=Random_1, children [WavePlayer_6,WavePlayer_7,WavePlayer_8], properties {bShouldExcludeFromBranchCulling=true, weights [1,1,1]}}
 {id=WavePlayer_6, properties {soundWaveAssetPtr=&/SpicyBerry_Land_01}}
 {id=WavePlayer_7, properties {soundWaveAssetPtr=&/SpicyBerry_Land_02}}
 {id=WavePlayer_8, properties {soundWaveAssetPtr=&/SpicyBerry_Land_03}}
]
properties {
 attenuationSettings=&Fort_Weapon_Drop_Att
 maxDistance=3000
 soundClassObject=&Fort_Loot_Collect
 volumeMultiplier=0.3
}
