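# Sound cue for picking up the SpicyBerry item.
# A Random node picks one of three equally weighted pickup wave variations and is excluded from branch culling.
# It uses weapon-drop attenuation and the loot-collect sound class at slightly reduced volume, with a concurrency override limiting playback to one instance per owner.
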SoundCue /SpicyBerry/Audio/SpicyBerry_Pickup_Cue
root=Random_0
nodes [
 {id=Random_0, children [WavePlayer_6,WavePlayer_7,WavePlayer_8], properties {bShouldExcludeFromBranchCulling=true, weights [1,1,1]}}
 {id=WavePlayer_6, properties {soundWaveAssetPtr=&/SpicyBerry_Pickup_01}}
 {id=WavePlayer_7, properties {soundWaveAssetPtr=&/SpicyBerry_Pickup_02}}
 {id=WavePlayer_8, properties {soundWaveAssetPtr=&/SpicyBerry_Pickup_03}}
]
properties {
 attenuationSettings=&Fort_Weapon_Drop_Att
 bOverrideConcurrency=true
 maxDistance=3000
 soundClassObject=&Fort_Loot_Collect
 volumeMultiplier=0.85
 concurrencyOverrides {bLimitToOwner=true, maxCount=1}
}
