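# Sound cue for the SpicyBerry throw-ready sound, which reuses three SpicyBerry landing waves.
# A Random node picks among the waves with equal weight and feeds a Modulator at the root, left at its usual settings.
# The cue plays at reduced volume through the weapon drop attenuation and the loot collect sound class.
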SoundCue /SpicyBerry/Audio/SpicyBerry_Throw_Ready_Cue
root=Modulator_2
nodes [
 {id=Modulator_2, children [Random_1]}
 {id=Random_1, children [WavePlayer_1,WavePlayer_2,WavePlayer_4], properties {bShouldExcludeFromBranchCulling=true, weights [1,1,1]}}
 {id=WavePlayer_1, properties {soundWaveAssetPtr=&/SpicyBerry_Land_03}}
 {id=WavePlayer_2, properties {soundWaveAssetPtr=&/SpicyBerry_Land_01}}
 {id=WavePlayer_4, properties {soundWaveAssetPtr=&/SpicyBerry_Land_02}}
]
properties {
 attenuationSettings=&Fort_Weapon_Drop_Att
 maxDistance=3000
 soundClassObject=&Fort_Loot_Collect
 volumeMultiplier=0.3
}
