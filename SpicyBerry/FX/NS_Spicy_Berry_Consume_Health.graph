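# Niagara one-shot burst effect for consuming a health Spicy Berry, drawing green-teal glows over the skeletal mesh supplied through User.SkeletalMesh.
# Decals001 places one sprite per spawn group on 17 bones or sockets with hand-tuned offsets, animating a dynamic material parameter and alpha.
# Flares001 layers four sprite renderers through custom curve-driven corona, anamorphic and distant-glow attributes, while Empty001 adds a brief fading light.
# The noise-driven Cast_Off_Bubbles emitter is disabled.
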NiagaraSystem /SpicyBerry/FX/NS_Spicy_Berry_Consume_Health
dependencies {
 usedDataInterfaces [&NiagaraDataInterfaceArrayFloat3, &NiagaraDataInterfaceCurve]
 usedDynamicInputs []
 usedModules [&SystemState]
}
emitters {
 Cast_Off_Bubbles {
  bEnabled=false
  emitterUpdateScript [
   &EmitterState(Life Cycle Mode=::NewEnumerator1, Loop Behavior=::NewEnumerator1, Loop Duration=1, UseLoopDelay=false)
   &SpawnBurst_Instantaneous(Spawn Count=75, Use Spawn Probability=false, UseLoopCountLimit=false)
  ]
  particleSpawnScript [
   &FN_SetSpawnGroup(Number of Spawn Groups=2)
   &InitializeParticle(
    Color Maximum=<rgba 1.0833 5 0 1>
    Color Minimum=<rgba 0.0213 1 0.3426 1>
    Color Mode=::NewEnumerator2
    Lifetime Max=1.5
    Lifetime Min=0.5
    Lifetime Mode=::NewEnumerator1
    Sprite Size Mode=::NewEnumerator1
    Uniform Sprite Size Min=4
    UsePositionOffset=false
    Write Material Random=true
   )
   &SkeletalMeshLocation(
    Bone / Socket Index=$Particles.SpawnGroup
    Bone Sampling Mode=::NewEnumerator4
    MaskSpawnGroup=false
    Sampled Position Offset=&SelectVectorFromArray(
     Direct Array Index=$Particles.SpawnGroup
     Vector Selection Array=DataInterface() {propertyValues {FloatData [
        <0 0 0>
        <5 0 0>
        <0 0 0>
        <0 0 0>
        <5 0 0>
        <0 0 0>
        <0 0 0>
        <-5 0 0>
        <-10 0 0>
        <8 0 0>
        <-8 0 0>
        <-4 0 0>
        <8 10 0>
        <8 0 0>
        <4 0 0>
        <-8 -10 0>
        <-10 -22 0>
       ]}}
    )
    Skeletal Mesh=$User.SkeletalMesh
   )
   &SphereLocation(
    MaskSpawnGroup=false
    Override Seed=false
    Sphere Radius=5
    Surface Only=false
    UseNonUniformScale=false
    UseOffset=false
    UseSphereOrientation=false
   )
   &AddVelocityFromPoint(
    Module.Use Falloff Distance=false
    Velocity Origin:NiagaraPosition=$Output.SphereLocation.Position
    Velocity Strength=&RandomRangeFloat(Minimum=100, Maximum=250)
   )
  ]
  particleUpdateScript [
   &ParticleState()
   &VectorNoiseForce(Force Amount=550, Module.Use Random Noise Time Interval=false)
   &Drag(Drag=&RandomRangeFloat(Minimum=1, Maximum=3), Use Linear Drag=true, Use Rotational Drag=false)
   &SolveForcesAndVelocity(Clamp Velocity=false, Limit Acceleration=false)
   &ScaleSpriteSize(
    Scale Sprite Size Mode=::NewEnumerator0
    Uniform Scale Factor=&FloatFromCurve(FloatCurve=DataInterface() {propertyValues {Curve {keys [
         {arriveTangentWeight=0.2716, interpMode=RCIM_Cubic, leaveTangentWeight=0.6337, tangentMode=RCTM_User, time=0.3, value=1}
         {arriveTangentWeight=0.6575, interpMode=RCIM_Cubic, tangentMode=RCTM_User, time=1, value=0}
        ]}}})
   )
  ]
  rendererClasses [&NiagaraSpriteRendererProperties]
  renderers [{rendererClass=&NiagaraSpriteRendererProperties}]
 }
 Decals001 {
  emitterUpdateScript [
   &EmitterState(Life Cycle Mode=::NewEnumerator1, Loop Behavior=::NewEnumerator1, Loop Duration=1, UseLoopDelay=false)
   &SpawnBurst_Instantaneous(Spawn Count=17, Use Spawn Probability=false, UseLoopCountLimit=false)
  ]
  particleSpawnScript [
   &FN_SetSpawnGroup(Number of Spawn Groups=17)
   &InitializeParticle(
    Color=<rgba 0.0213 1 0.3426 0.5>
    Sprite Size Mode=::NewEnumerator3
    Uniform Sprite Size=40
    UsePositionOffset=false
    Write Material Random=true
   )
   &SkeletalMeshLocation(
    Bone / Socket Index=$Particles.SpawnGroup
    Bone Sampling Mode=::NewEnumerator4
    MaskSpawnGroup=false
    Sampled Position Offset=&SelectVectorFromArray(
     Direct Array Index=$Particles.SpawnGroup
     Vector Selection Array=DataInterface() {propertyValues {FloatData [
        <0 0 0>
        <5 0 0>
        <0 0 0>
        <0 0 0>
        <5 0 0>
        <0 0 0>
        <0 0 0>
        <-5 0 0>
        <-10 0 0>
        <8 0 0>
        <-8 0 0>
        <-4 0 0>
        <8 10 0>
        <8 0 0>
        <4 0 0>
        <-8 -10 0>
        <-10 -22 0>
       ]}}
    )
    Skeletal Mesh=$User.SkeletalMesh
   )
   &CameraOffset(Camera Offset Amount=25)
  ]
  particleUpdateScript [
   &ParticleState()
   &DynamicMaterialParameters(
    Index 0 Param 1=&FloatFromCurve(
     FloatCurve=DataInterface() {propertyValues {Curve {keys [
         {interpMode=RCIM_Cubic, leaveTangentWeight=0.2595, tangentMode=RCTM_User, time=0, value=-0.5}
         {arriveTangent=21.8336, interpMode=RCIM_Cubic, leaveTangent=21.8336, tangentMode=RCTM_User, time=0.075, value=0.275}
         {arriveTangent=5.1081, interpMode=RCIM_Cubic, leaveTangent=5.1081, time=0.15, value=1}
         {arriveTangent=4.7059, interpMode=RCIM_Cubic, leaveTangent=4.7059, time=1, value=5}
        ]}}}
     Scale Curve=4
    )
    Module.Param0WriteEnabled=true
    Module.Param1WriteEnabled=false
    Module.Param2WriteEnabled=false
    Module.Param3WriteEnabled=false
   )
   &ScaleColor(
    Scale Alpha=&FloatFromCurve(FloatCurve=DataInterface() {propertyValues {Curve {keys [
         {arriveTangent=9.0725, interpMode=RCIM_Cubic, leaveTangent=9.0725, tangentMode=RCTM_User, time=0, value=0}
         {arriveTangentWeight=0.181, interpMode=RCIM_Cubic, leaveTangentWeight=0.7242, tangentMode=RCTM_User, time=0.2, value=1}
         {arriveTangent=-0.0405, interpMode=RCIM_Cubic, leaveTangent=-0.0405, tangentMode=RCTM_User, time=1, value=0}
        ]}}})
    Scale RGB=&VectorFromFloat(Value=0.15)
    ScaleA=true
    ScaleRGB=true
   )
  ]
  rendererClasses [&NiagaraSpriteRendererProperties]
  renderers [{rendererClass=&NiagaraSpriteRendererProperties}]
 }
 Empty001 {
  emitterUpdateScript [
   &EmitterState(Life Cycle Mode=::NewEnumerator1, Loop Behavior=::NewEnumerator1, Loop Duration=1, UseLoopDelay=false)
   &SpawnBurst_Instantaneous(Use Spawn Probability=false, UseLoopCountLimit=false)
  ]
  particleSpawnScript [
   &InitializeParticle(Color=<rgba 0.0213 1 0.3426 1>, Lifetime=0.2, Position Offset=<0 0 -50>, UsePositionOffset=true, Write Material Random=true)
   &SkeletalMeshLocation(MaskSpawnGroup=false, Mesh Sampling Type=::NewEnumerator5, OverrideRandomSeed=false, Skeletal Mesh=$User.SkeletalMesh)
  ]
  particleUpdateScript [
   &ParticleState()
   /SetVariables(Particles.LightExponent=2, Particles.LightRadius=150) {isSetParameters=true}
   &ScaleColor(
    Scale RGB=&VectorFromFloat(Value=&FloatFromCurve(
      FloatCurve=DataInterface() {propertyValues {Curve {keys [
          {interpMode=RCIM_Cubic, leaveTangentWeight=0.4714, tangentMode=RCTM_User, time=0, value=1}
          {arriveTangentWeight=0.4714, interpMode=RCIM_Cubic, tangentMode=RCTM_User, time=1, value=0}
         ]}}}
      Scale Curve=5
     ))
    ScaleA=true
    ScaleRGB=true
   )
  ]
  rendererClasses [&NiagaraLightRendererProperties]
  renderers [{rendererClass=&NiagaraLightRendererProperties, rendererIndex=0}]
 }
 Flares001 {
  emitterUpdateScript [
   &EmitterState(Life Cycle Mode=::NewEnumerator1, Loop Behavior=::NewEnumerator1, Loop Duration=1, UseLoopDelay=false)
   &SpawnBurst_Instantaneous(Use Spawn Probability=false, UseLoopCountLimit=false)
  ]
  particleSpawnScript [
   &InitializeParticle(
    Color=<rgba 0 3.434 1.1275 2>
    Lifetime=0.2
    Sprite Rotation Angle=0.2
    Sprite Rotation Mode=::NewEnumerator1
    Sprite Size Mode=::NewEnumerator3
    Uniform Sprite Size=150
    UsePositionOffset=false
    Write Material Random=true
   )
   &CameraOffset(Camera Offset Amount=75)
   &SkeletalMeshLocation(MaskSpawnGroup=false, Mesh Sampling Type=::NewEnumerator5, OverrideRandomSeed=false, Skeletal Mesh=$User.SkeletalMesh)
   /SetVariables(Particles.AnamorphicSize:Vector2f=<240 7>) {isSetParameters=true}
  ]
  particleUpdateScript [
   &ParticleState()
   &ScaleColor(
    Scale Alpha=&FloatFromCurve(FloatCurve=DataInterface() {propertyValues {Curve {keys [
         {arriveTangent=0.6364, interpMode=RCIM_Cubic, leaveTangent=0.6364, time=0, value=0}
         {arriveTangent=1.4286, interpMode=RCIM_Cubic, leaveTangent=1.4286, time=0.55, value=0.35}
         {arriveTangent=-8.2069, interpMode=RCIM_Cubic, leaveTangent=-8.2069, tangentMode=RCTM_User, time=0.7, value=1}
         {arriveTangent=-0.0405, interpMode=RCIM_Cubic, leaveTangent=-0.0405, tangentMode=RCTM_User, time=1, value=0}
        ]}}})
    ScaleA=true
    ScaleRGB=true
   )
   &ScaleSpriteSize(Uniform Curve Sprite Scale=DataInterface() {propertyValues {Curve {keys [
        {interpMode=RCIM_Cubic, leaveTangentWeight=0.1863, tangentMode=RCTM_User, time=0, value=0.4}
        {
         arriveTangent=0.3825
         arriveTangentWeight=0.2415
         interpMode=RCIM_Cubic
         leaveTangent=0.3825
         leaveTangentWeight=0.0659
         tangentMode=RCTM_User
         time=0.55
         value=0.5
        }
        {
         arriveTangent=-1.9384
         arriveTangentWeight=0.0547
         interpMode=RCIM_Cubic
         leaveTangent=-1.9384
         leaveTangentWeight=0.1094
         tangentMode=RCTM_User
         time=0.7
         value=1
        }
        {arriveTangentWeight=0.1414, interpMode=RCIM_Cubic, tangentMode=RCTM_User, time=1, value=0.7}
       ]}}})
   /SetVariables(
    Particles.Anamorphic Color:LinearColor=&MakeLinearColorFromVectorAndFloat(
     "Vector (RGB)"=&MakeVectorFromLinearColorRGB(Color=<rgba 0 1.2 0.394 1>)
     "Float (Alpha)"=&Multiply_Float(
      A=0.2
      B=&FloatFromCurve(FloatCurve=DataInterface() {propertyValues {Curve {keys [
           {arriveTangent=0.6364, interpMode=RCIM_Cubic, leaveTangent=0.6364, time=0, value=0}
           {arriveTangent=1.4286, interpMode=RCIM_Cubic, leaveTangent=1.4286, time=0.55, value=0.35}
           {arriveTangent=-8.2069, interpMode=RCIM_Cubic, leaveTangent=-8.2069, tangentMode=RCTM_User, time=0.7, value=1}
           {arriveTangent=-0.0405, interpMode=RCIM_Cubic, leaveTangent=-0.0405, tangentMode=RCTM_User, time=1, value=0}
          ]}}})
     )
    )
    Particles.CoronaColor:LinearColor=&MakeLinearColorFromVectorAndFloat(
     "Vector (RGB)"=&MakeVectorFromLinearColorRGB(Color=<rgba 0 1.2 0.394 1>)
     "Float (Alpha)"=&FloatFromCurve(FloatCurve=DataInterface() {propertyValues {Curve {keys [
          {interpMode=RCIM_Cubic, leaveTangentWeight=0.3578, tangentMode=RCTM_User, time=0, value=0}
          {arriveTangent=0.6909, interpMode=RCIM_Cubic, leaveTangent=0.6909, tangentMode=RCTM_User, time=0.5, value=0.95}
          {arriveTangentWeight=0.1431, interpMode=RCIM_Cubic, leaveTangentWeight=0.2147, tangentMode=RCTM_User, time=0.7, value=1}
          {arriveTangentWeight=0.348, interpMode=RCIM_Cubic, tangentMode=RCTM_User, time=1, value=0}
         ]}}})
    )
    Particles.CoronaDynParameters:Vector4f=&MakeVector4(
     X=&FloatFromCurve(FloatCurve=DataInterface() {propertyValues {Curve {keys [
          {interpMode=RCIM_Cubic, leaveTangentWeight=0.1904, tangentMode=RCTM_User, time=0, value=1.4}
          {
           arriveTangent=-1.2888
           arriveTangentWeight=0.2753
           interpMode=RCIM_Cubic
           leaveTangent=-1.2888
           leaveTangentWeight=0.0918
           tangentMode=RCTM_User
           time=0.525
           value=1.175
          }
          {arriveTangent=-1.9474, interpMode=RCIM_Cubic, leaveTangent=-1.9474, time=0.7, value=0.55}
          {arriveTangentWeight=0.1414, interpMode=RCIM_Cubic, tangentMode=RCTM_User, time=1, value=0.25}
         ]}}})
     Y=8
     Z=0.6
     W=1
    )
    Particles.CoronaSize:Vector2f=&Vector2DFromFloat(Value=&Multiply_Float(
      A=75
      B=&FloatFromCurve(FloatCurve=DataInterface() {propertyValues {Curve {keys [
           {interpMode=RCIM_Cubic, leaveTangentWeight=0.3578, tangentMode=RCTM_User, time=0, value=0}
           {
            arriveTangent=1.1981
            arriveTangentWeight=0.2906
            interpMode=RCIM_Cubic
            leaveTangent=1.1981
            leaveTangentWeight=0.1163
            tangentMode=RCTM_User
            time=0.5
            value=0.95
           }
           {arriveTangentWeight=0.1431, interpMode=RCIM_Cubic, leaveTangentWeight=0.2147, tangentMode=RCTM_User, time=0.7, value=1}
          ]}}})
     ))
    Particles.DistanceGlowSize:Vector2f=&Vector2DFromFloat(Value=&Multiply_Float(
      A=350
      B=&FloatFromCurve(FloatCurve=DataInterface() {propertyValues {Curve {keys [
           {interpMode=RCIM_Cubic, leaveTangentWeight=0.348, tangentMode=RCTM_User, time=0, value=0.7}
           {arriveTangent=0.7107, arriveTangentWeight=0.348, interpMode=RCIM_Cubic, leaveTangent=0.7107, tangentMode=RCTM_User, time=1, value=1}
          ]}}})
     ))
    Particles.DistantGlowColor:LinearColor=&MakeLinearColorFromVectorAndFloat(
     "Vector (RGB)"=&MakeVectorFromLinearColorRGB(Color=<rgba 0.0213 1 0.3426 1>)
     "Float (Alpha)"=&FloatFromCurve(FloatCurve=DataInterface() {propertyValues {Curve {keys [
          {arriveTangent=9.0725, interpMode=RCIM_Cubic, leaveTangent=9.0725, tangentMode=RCTM_User, time=0, value=0}
          {arriveTangentWeight=0.0905, interpMode=RCIM_Cubic, leaveTangentWeight=0.8147, tangentMode=RCTM_User, time=0.1, value=0.3}
          {arriveTangent=-0.0405, interpMode=RCIM_Cubic, leaveTangent=-0.0405, tangentMode=RCTM_User, time=1, value=0}
         ]}}})
    )
   ) {isSetParameters=true}
  ]
  rendererClasses [&NiagaraSpriteRendererProperties]
  renderers [
   {rendererClass=&NiagaraSpriteRendererProperties}
   {rendererClass=&NiagaraSpriteRendererProperties, rendererIndex=1}
   {rendererClass=&NiagaraSpriteRendererProperties, rendererIndex=2}
   {rendererClass=&NiagaraSpriteRendererProperties, rendererIndex=3}
  ]
 }
}
summary {
 systemName=NS_Spicy_Berry_Consume_Health
 emitters [
  {bEnabled=false, emitterName=Cast_Off_Bubbles, rendererClasses [&NiagaraSpriteRendererProperties]}
  {emitterName=Decals001, rendererClasses [&NiagaraSpriteRendererProperties]}
  {emitterName=Flares001, rendererClasses [&NiagaraSpriteRendererProperties]}
  {emitterName=Empty001, rendererClasses [&NiagaraLightRendererProperties]}
 ]
 userVariables [User.SkeletalMesh {
   type=NiagaraDataInterfaceSkeletalMesh
   defaultValue {
    struct=NiagaraExt_VariableValue_DataInterface
    value {dataInterface=&self:NiagaraDataInterfaceSkeletalMesh_0}
   }
  }]
}
